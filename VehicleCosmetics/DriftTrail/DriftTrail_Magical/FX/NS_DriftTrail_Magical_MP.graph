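# Magical drift trail vehicle cosmetic, a Niagara system with ribbon, sprite/mesh and wheel-wrap emitters at the wheels.
# Two short-lived ribbons stream backward in four layered renderers, while persistent burst particles attach to skeletal mesh sockets, with wrap meshes scaled per wheel.
# Painted-variant modules and the bIsPainted flag switch each layer between default magical colors and paint colors.
# System-level vehicle flag and speed modules provide vehicle state, and each emitter runs per-wheel tire FX disabling.
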NiagaraSystem /VehicleCosmetics/DriftTrail/DriftTrail_Magical/FX/NS_DriftTrail_Magical_MP
dependencies {
 usedDataInterfaces [&NiagaraDataInterfaceColorCurve, &NiagaraDataInterfaceCurve, &NiagaraDataInterfaceSkeletalMesh]
 usedDynamicInputs []
 usedModules [
  &self:SystemSpawnScript.SystemScriptSource.SystemScriptGraph.NiagaraNodeAssignment_0.SetVariables_#1
  &SystemState
  &FN_UnpackVehicleFlags
  /VehicleCosmetics/FX/Niagara/Modules/VC_SetVehicleSpeedAttribs
  /Game/Effects/Niagara/Modules/Vehicle/FN_VehicleFlags_Cheatcodes
 ]
}
emitters {
 NE_VC_DriftTrail_Ribbon_Parent {
  emitterUpdateScript [
   &EmitterState()
   &VC_RateQualityHandler(Epic Spawn Rate=90, High Spawn Rate=60, Low Spawn Rate=30, Medium Spawn Rate=45, Use Custom Rate?=true)
   &VC_SpawnRibbon_v1(Number of Ribbons=2, SpawnRate=$Emitter.Spawn Rate)
  ]
  particleSpawnScript [
   &VC_PaintedVariantsOverride()
   &VC_InitializeRibbon_v1(
    Color=$Particles.Color
    Custom Offset=true
    Enable Skeletal Position=true
    Lifetime=0.6
    Local Vector Offset=<-50 0 -20>
    Offset Amount=1
    Ribbon Width=25
   )
   &AddVelocity(Velocity=<-1500 0 0>, Velocity Speed Scale=2) {module=AddVelocity001}
  ]
  particleUpdateScript [
   &FN_DisableTireFX(Mask Index Attribute:NiagaraInt32=$Particles.SpawnGroup, Mask Index FL=2, Mask Index FR=3, Mask Index RL=0, Mask Index RR=1)
   &VC_PaintedVariantsOverride(
    BurntSienna=<rgba 0.0642 0.0198 0.0079 1>
    Cobalt=<rgba 0 0.0185 1 1>
    Crimson=<rgba 0.4635 0 0.0126 1>
    ForestGreen=<rgba 0 0.5 0.0253 1>
    Grey=<rgba 0.0781 0.0781 0.0781 1>
    Lime=<rgba 0.2943 0.5885 0 1>
    Orange=<rgba 0.6724 0.2346 0 1>
    Purple=<rgba 0.0844 0.0482 0.3372 1>
    R1 Black:LinearColor=&ColorFromCurve(ColorCurve=DataInterface() {propertyValues {
       AlphaCurve {keys [<at 0 1 RCIM_Cubic>,<at 1 0 RCIM_Cubic>]}
       BlueCurve {keys [<at 0 0.0521 RCIM_Cubic>,<at 0.1166 0.01 RCIM_Cubic>]}
       GreenCurve {keys [<at 0 0.0521 RCIM_Cubic>,<at 0.1166 0.01 RCIM_Cubic>]}
       RedCurve {keys [<at 0 0.0521 RCIM_Cubic>,<at 0.1166 0.01 RCIM_Cubic>]}
      }})
    R1 BurntSienna:LinearColor=&ColorFromCurve(ColorCurve=DataInterface() {propertyValues {
       AlphaCurve {keys [<at 0 1 RCIM_Cubic>,<at 1 0 RCIM_Cubic>]}
       BlueCurve {keys [<at 0 0.0482 RCIM_Cubic>,<at 0.2929 0.0107 RCIM_Cubic>]}
       GreenCurve {keys [<at 0 0.0887 RCIM_Cubic>,<at 0.2929 0.0196 RCIM_Cubic>]}
       RedCurve {keys [<at 0 0.4233 RCIM_Cubic>,<at 0.2929 0.0938 RCIM_Cubic>]}
      }})
    R1 Cobalt:LinearColor=&ColorFromCurve(ColorCurve=DataInterface() {propertyValues {
       AlphaCurve {keys [<at 0 1 RCIM_Cubic>,<at 1 0 RCIM_Cubic>]}
       BlueCurve {keys [<at 0 0.3594 RCIM_Cubic>,<at 0.2929 0.7231 RCIM_Cubic>]}
       GreenCurve {keys [<at 0 0.115 RCIM_Cubic>,<at 0.2929 0.1022 RCIM_Cubic>]}
       RedCurve {keys [<at 0 0.1104 RCIM_Cubic>,<at 0.2929 0.0056 RCIM_Cubic>]}
      }})
    R1 Crimson:LinearColor=&ColorFromCurve(ColorCurve=DataInterface() {propertyValues {
       AlphaCurve {keys [<at 0 1 RCIM_Cubic>,<at 1 0 RCIM_Cubic>]}
       BlueCurve {keys [<at 0 0.0755 RCIM_Cubic>,<at 0.2929 0.0123 RCIM_Cubic>]}
       GreenCurve {keys [<at 0 0.065 RCIM_Cubic>,<at 0.2929 0 RCIM_Cubic>]}
       RedCurve {keys [<at 0 0.4621 RCIM_Cubic>,<at 0.2929 0.4621 RCIM_Cubic>]}
      }})
    R1 ForestGreen:LinearColor=&ColorFromCurve(ColorCurve=DataInterface() {propertyValues {
       AlphaCurve {keys [<at 0 1 RCIM_Cubic>,<at 1 0 RCIM_Cubic>]}
       BlueCurve {keys [<at 0 0.1981 RCIM_Cubic>,<at 0.2929 0.028 RCIM_Cubic>]}
       GreenCurve {keys [<at 0 0.5781 RCIM_Cubic>,<at 0.2929 0.5029 RCIM_Cubic>]}
       RedCurve {keys [<at 0 0.2223 RCIM_Cubic>,<at 0.2929 0.1087 RCIM_Cubic>]}
      }})
    R1 Grey:LinearColor=&ColorFromCurve(ColorCurve=DataInterface() {propertyValues {
       AlphaCurve {keys [<at 0 1 RCIM_Cubic>,<at 1 0 RCIM_Cubic>]}
       BlueCurve {keys [<at 0 0.151 RCIM_Cubic>,<at 0.2929 0.0521 RCIM_Cubic>]}
       GreenCurve {keys [<at 0 0.151 RCIM_Cubic>,<at 0.2929 0.0521 RCIM_Cubic>]}
       RedCurve {keys [<at 0 0.151 RCIM_Cubic>,<at 0.2929 0.0521 RCIM_Cubic>]}
      }})
    R1 Lime:LinearColor=&ColorFromCurve(ColorCurve=DataInterface() {propertyValues {
       AlphaCurve {keys [<at 0 1 RCIM_Cubic>,<at 1 0 RCIM_Cubic>]}
       BlueCurve {keys [<at 0 0.0431 RCIM_Cubic>,<at 0.2126 0 RCIM_Cubic>]}
       GreenCurve {keys [<at 0 0.5906 RCIM_Cubic>,<at 0.2126 0.3958 RCIM_Cubic>]}
       RedCurve {keys [<at 0 0.3176 RCIM_Cubic>,<at 0.2126 0.1985 RCIM_Cubic>]}
      }})
    R1 Orange:LinearColor=&ColorFromCurve(ColorCurve=DataInterface() {propertyValues {
       AlphaCurve {keys [<at 0 1 RCIM_Cubic>,<at 1 0 RCIM_Cubic>]}
       BlueCurve {keys [<at 0 0 RCIM_Cubic>,<at 0.2126 0.0028 RCIM_Cubic>]}
       GreenCurve {keys [<at 0 0.2335 RCIM_Cubic>,<at 0.2126 0.0221 RCIM_Cubic>]}
       RedCurve {keys [<at 0 0.707 RCIM_Cubic>,<at 0.2126 0.2917 RCIM_Cubic>]}
      }})
    R1 Pink:LinearColor=&ColorFromCurve(ColorCurve=DataInterface() {propertyValues {
       AlphaCurve {keys [<at 0 1 RCIM_Cubic>,<at 1 0 RCIM_Cubic>]}
       BlueCurve {keys [<at 0 0.237 RCIM_Cubic>,<at 0.2126 0.3419 RCIM_Cubic>]}
       GreenCurve {keys [<at 0 0.1507 RCIM_Cubic>,<at 0.2126 0.0612 RCIM_Cubic>]}
       RedCurve {keys [<at 0 0.2865 RCIM_Cubic>,<at 0.2126 0.5029 RCIM_Cubic>]}
      }})
    R1 Purple:LinearColor=&ColorFromCurve(ColorCurve=DataInterface() {propertyValues {
       AlphaCurve {keys [<at 0 1 RCIM_Cubic>,<at 1 0 RCIM_Cubic>]}
       BlueCurve {keys [<at 0 0.3372 RCIM_Cubic>,<at 0.2126 0.4323 RCIM_Cubic>]}
       GreenCurve {keys [<at 0 0.1106 RCIM_Cubic>,<at 0.2126 0.0316 RCIM_Cubic>]}
       RedCurve {keys [<at 0 0.139 RCIM_Cubic>,<at 0.2126 0.0616 RCIM_Cubic>]}
      }})
    R1 Saffron:LinearColor=&ColorFromCurve(ColorCurve=DataInterface() {propertyValues {
       AlphaCurve {keys [<at 0 1 RCIM_Cubic>,<at 1 0 RCIM_Cubic>]}
       BlueCurve {keys [<at 0 0.1051 RCIM_Cubic>,<at 0.2126 0 RCIM_Cubic>]}
       GreenCurve {keys [<at 0 0.252 RCIM_Cubic>,<at 0.2126 0.5227 RCIM_Cubic>]}
       RedCurve {keys [<at 0 0.2969 RCIM_Cubic>,<at 0.2126 0.6823 RCIM_Cubic>]}
      }})
    R1 SkyBlue:LinearColor=&ColorFromCurve(ColorCurve=DataInterface() {propertyValues {
       AlphaCurve {keys [<at 0 1 RCIM_Cubic>,<at 1 0 RCIM_Cubic>]}
       BlueCurve {keys [<at 0 0.4323 RCIM_Cubic>,<at 0.2126 0.3698 RCIM_Cubic>]}
       GreenCurve {keys [<at 0 0.3419 RCIM_Cubic>,<at 0.2126 0.2471 RCIM_Cubic>]}
       RedCurve {keys [<at 0 0.1599 RCIM_Cubic>,<at 0.2126 0 RCIM_Cubic>]}
      }})
    R1 TitaniumWhite:LinearColor=&ColorFromCurve(ColorCurve=DataInterface() {propertyValues {
       AlphaCurve {keys [<at 0 0.9 RCIM_Cubic>,<at 1 0 RCIM_Cubic>]}
       BlueCurve {keys [<at 0 0.3712 RCIM_Cubic>,<at 0.2943 0.8125 RCIM_Cubic>]}
       GreenCurve {keys [<at 0 0.3712 RCIM_Cubic>,<at 0.2943 0.8125 RCIM_Cubic>]}
       RedCurve {keys [<at 0 0.3712 RCIM_Cubic>,<at 0.2943 0.8125 RCIM_Cubic>]}
      }})
    R2 Black:LinearColor=<rgba 0.1094 0.1094 0.1094 1>
    R2 BurntSienna:LinearColor=<rgba 0.4219 0.1303 0.052 1>
    R2 Cobalt:LinearColor=<rgba 0.0625 0.0798 1 1>
    R2 Crimson:LinearColor=<rgba 0.3802 0.0356 0.0448 1>
    R2 Grey:LinearColor=<rgba 0.0417 0.0417 0.0417 1>
    R2 Purple:LinearColor=<rgba 0.0844 0.0482 0.3372 1>
    Saffron=<rgba 0.2708 0.2076 0 1>
    SkyBlue=<rgba 0 0.357 0.5365 1>
    TitaniumWhite=<rgba 1.5 1.5 1.5 1>
    Use R1 Painted Overrides=true
    Use R2 Painted Overrides=true
   ) {module=VC_PaintedVariantsOverride001}
   &ParticleState()
   &InheritSourceMovement(
    Applied Position Scale=&VectorFromFloat(Value=&FloatFromCurve(FloatCurve=DataInterface() {propertyValues {Curve {keys [<at 0 1 RCIM_Cubic>,<at 1 0 RCIM_Cubic>]}}}))
    ApplyForce=false
    ApplyPosition=true
    ApplyPositionFalloff=false
   )
   &SolveForcesAndVelocity(Clamp Velocity=false, Limit Acceleration=false)
   &VC_UpdateRibbon_v1(Enable Skeletal Position=true)
   &DynamicMaterialParameters(
    Index 0 Param 1=2
    Index 0 Param 2=0.5
    Index 0 Param 3=1
    Index 0 Param 4=1
    Module.Index0Float0Write=true
    Module.Index0Float1Write=true
    Module.Index0Float2Write=true
    Module.Index0Float3Write=true
   )
   /SetVariables(
    Particles.Index1:Vector4f=&MakeVector4(X=3, Y=3, Z=1, W=-1)
    Particles.Index2:Vector4f=&MakeVector4(X=2, Y=1, Z=0, W=0)
    Particles.Index3:Vector4f=&MakeVector4(X=2, Y=&MakeCustomFloatFromBool(Bool=$User.bIsPainted, True Float=300, False Float=100), Z=8, W=-1)
    Particles.RibbonColor0:LinearColor=&VC_DynamicScaleColor(
     Color to Scale:LinearColor=$Particles.Color
     RGB Scale:Vector3f=&MakeCustomVectorFromBool(
      Bool=$User.bIsPainted
      True Vector=&VectorFromFloat(Value=&Multiply_Float(A=0.1, B=0.4))
      False Vector=&MakeVectorFromLinearColorRGB(Color=&Multiply_LinearColorByFloat(Linear Color=<rgba 0.08 0.05 0.1 1>, Float=0.4))
     )
     Alpha Scale:NiagaraFloat=&FloatFromCurve(
      FloatCurve=DataInterface() {propertyValues {Curve {keys [
          <at 0 0.5 RCIM_Cubic>
          {arriveTangentWeight=0.1572, interpMode=RCIM_Cubic, leaveTangentWeight=0.1572, tangentMode=RCTM_User, time=0.3, value=1}
          <at 0.7 0 RCIM_Cubic>
         ]}}}
      Scale Curve=0.8
     )
    )
    Particles.RibbonColor1:LinearColor=&VC_DynamicScaleColor(
     Color to Scale:LinearColor=&MakeCustomLinearColorFromBool(
      Bool=$User.bIsPainted
      True Color=$Particles.R1 Color
      False Color=&ColorFromCurve(ColorCurve=DataInterface() {propertyValues {
         AlphaCurve {keys [<at 0 1 RCIM_Cubic>,<at 1 1 RCIM_Cubic>]}
         BlueCurve {keys [
           <at 0.0459 0.1458 RCIM_Cubic>
           {arriveTangentWeight=0.1346, interpMode=RCIM_Cubic, leaveTangentWeight=0.1565, tangentMode=RCTM_User, time=0.1, value=1}
           <at 0.25 1 RCIM_Cubic>
          ]}
         GreenCurve {keys [
           <at 0.0459 1 RCIM_Cubic>
           {arriveTangent=-4.5905, interpMode=RCIM_Cubic, leaveTangent=-4.5905, time=0.1, value=0.3005}
           <at 0.25 0.063 RCIM_Cubic>
          ]}
         RedCurve {keys [
           <at 0.0459 1 RCIM_Cubic>
           {arriveTangentWeight=0.1339, interpMode=RCIM_Cubic, leaveTangentWeight=0.1557, tangentMode=RCTM_User, time=0.1, value=1}
           <at 0.25 0.15 RCIM_Cubic>
          ]}
        }})
     )
     RGB Scale:Vector3f=&MakeCustomVectorFromBool(
      Bool=$User.bIsPainted
      True Vector=&VectorFromFloat(Value=&FloatFromCurve(FloatCurve=DataInterface() {propertyValues {Curve {keys [<at 0 5 RCIM_Cubic>,<at 0.3 0.5 RCIM_Cubic>]}}}))
      False Vector=&VectorFromFloat(Value=&FloatFromCurve(FloatCurve=DataInterface() {propertyValues {Curve {keys [<at 0 2 RCIM_Cubic>,<at 0.3 0.5 RCIM_Cubic>]}}}))
     )
     Alpha Scale:NiagaraFloat=&FloatFromCurve(FloatCurve=DataInterface() {propertyValues {Curve {keys [
          <at 0 0 RCIM_Cubic>
          {arriveTangentWeight=0.2247, interpMode=RCIM_Cubic, leaveTangentWeight=0.4494, tangentMode=RCTM_User, time=0.1, value=2}
          {arriveTangentWeight=0.0667, interpMode=RCIM_Cubic, leaveTangentWeight=0.0667, tangentMode=RCTM_User, time=0.3, value=2}
          {arriveTangentWeight=0.7723, interpMode=RCIM_Cubic, tangentMode=RCTM_User, time=0.5, value=0}
         ]}}})
    )
    Particles.RibbonColor3:LinearColor=&VC_DynamicScaleColor(
     Color to Scale:LinearColor=&MakeCustomLinearColorFromBool(
      Bool=$User.bIsPainted
      True Color=$Particles.R1 Color
      False Color=&ColorFromCurve(ColorCurve=DataInterface() {propertyValues {
         AlphaCurve {keys [<at 0 1 RCIM_Cubic>,<at 1 1 RCIM_Cubic>]}
         BlueCurve {keys [<at 0.0459 0.1458 RCIM_Cubic>,<at 0.1295 1>,<at 0.2417 0.8542>]}
         GreenCurve {keys [<at 0.0459 0.9372 RCIM_Cubic>,<at 0.1295 0.3005>,<at 0.2417 0.0623>]}
         RedCurve {keys [<at 0.0459 1 RCIM_Cubic>,<at 0.1295 1>,<at 0.2417 0.1503>]}
        }})
     )
     RGB Scale:Vector3f=&MakeCustomVectorFromBool(
      Bool=$User.bIsPainted
      True Vector=&VectorFromFloat(Value=&FloatFromCurve(FloatCurve=DataInterface() {propertyValues {Curve {keys [<at 0 50>,<at 0.3 0.5>]}}}))
      False Vector=&VectorFromFloat(Value=1)
     )
     Alpha Scale:NiagaraFloat=&FloatFromCurve(
      FloatCurve=DataInterface() {propertyValues {Curve {keys [
          {arriveTangent=11.5571, interpMode=RCIM_Cubic, leaveTangent=11.5571, tangentMode=RCTM_User, time=0, value=0}
          <at 0.25 1 RCIM_Cubic>
          {arriveTangent=-9.2133, interpMode=RCIM_Cubic, leaveTangent=-9.2133, tangentMode=RCTM_User, time=0.5, value=0}
         ]}}}
      Scale Curve=0.5
     )
    )
    Particles.RibbonWidth0:NiagaraFloat=&Multiply_Float(
     A=&FloatFromCurve(FloatCurve=DataInterface() {propertyValues {Curve {keys [<at 0 1>,<at 1 0>]}}}, Scale Curve=3.5)
     B=$Particles.RibbonWidth
    )
    Particles.RibbonWidth1:NiagaraFloat=&Multiply_Float(
     A=&FloatFromCurve(FloatCurve=DataInterface() {propertyValues {Curve {keys [
          {arriveTangent=-1.5, interpMode=RCIM_Cubic, time=0, value=0.4}
          {arriveTangentWeight=0.1221, interpMode=RCIM_Cubic, leaveTangentWeight=0.2848, tangentMode=RCTM_User, time=0.2, value=0.8}
          <at 0.5 0.2 RCIM_Cubic>
         ]}}})
     B=$Particles.RibbonWidth
    )
    Particles.RibbonWidth2:NiagaraFloat=&Multiply_Float(A=1.5, B=$Particles.RibbonWidth)
    Particles.RibbonWidth3:NiagaraFloat=&Multiply_Float(
     A=&FloatFromCurve(FloatCurve=DataInterface() {propertyValues {Curve {keys [<at 0 0.7>,<at 1 0>]}}})
     B=$Particles.RibbonWidth
    )
    Particles.Ribboncolor2:LinearColor=&VC_DynamicScaleColor(
     Color to Scale:LinearColor=&MakeCustomLinearColorFromBool(
      Bool=$User.bIsPainted
      True Color=$Particles.R1 Color
      False Color=&ColorFromCurve(ColorCurve=DataInterface() {propertyValues {
         AlphaCurve {keys [<at 0 1 RCIM_Cubic>,<at 1 1 RCIM_Cubic>]}
         BlueCurve {keys [<at 0.1234 0.4 RCIM_Cubic>,<at 0.1866 1>,<at 0.2559 1 RCIM_Cubic>]}
         GreenCurve {keys [<at 0.1234 1 RCIM_Cubic>,<at 0.1866 0.3>,<at 0.2559 0.1 RCIM_Cubic>]}
         RedCurve {keys [<at 0.1234 1 RCIM_Cubic>,<at 0.1866 1>,<at 0.2559 0.2 RCIM_Cubic>]}
        }})
     )
     RGB Scale:Vector3f=&VectorFromFloat(Value=&FloatFromCurve(FloatCurve=DataInterface() {propertyValues {Curve {keys [<at 0.2 1>,<at 0.5 2>]}}}))
     Alpha Scale:NiagaraFloat=&FloatFromCurve(
      FloatCurve=DataInterface() {propertyValues {Curve {keys [
          <at 0 0 RCIM_Cubic>
          {interpMode=RCIM_Cubic, leaveTangentWeight=0.0833, tangentMode=RCTM_User, time=0.2, value=1}
          {arriveTangentWeight=0.1008, interpMode=RCIM_Cubic, tangentMode=RCTM_User, time=0.5, value=0}
         ]}}}
      Scale Curve=4
     )
    )
   ) {isSetParameters=true}
  ]
  rendererClasses [&NiagaraRibbonRendererProperties]
  renderers [
   {rendererClass=&NiagaraRibbonRendererProperties}
   {rendererClass=&NiagaraRibbonRendererProperties, rendererIndex=1}
   {rendererClass=&NiagaraRibbonRendererProperties, rendererIndex=2}
   {rendererClass=&NiagaraRibbonRendererProperties, rendererIndex=3}
  ]
 }
 NE_VC_DriftTrail_Sprite_Parent {
  emitterUpdateScript [
   &EmitterState(
    Inactive Response=::NewEnumerator1
    Life Cycle Mode=::NewEnumerator1
    Loop Behavior=::NewEnumerator1
    Loop Duration Mode=::NewEnumerator1
    UseLoopDelay=false
   )
   &SpawnBurst_Instantaneous(Use Spawn Probability=false, UseLoopCountLimit=false)
   &SpawnBurst_Instantaneous(Spawn Group=1, Use Spawn Probability=false, UseLoopCountLimit=false) {module=SpawnBurst_Instantaneous001}
   &SpawnBurst_Instantaneous(Spawn Group=2, Use Spawn Probability=false, UseLoopCountLimit=false) {enabled=false, module=SpawnBurst_Instantaneous002}
   &SpawnRate(SpawnGroup=3, SpawnRate=5, Use Spawn Probability=false) {enabled=false, module=SpawnRate003}
   &SpawnRate(
    SpawnGroup=3
    SpawnRate=/VehicleCosmetics/FX/Niagara/Modules/VC_Mask_Floats(
     FrontEnd Value:NiagaraFloat=200
     blsFrontEndPreview:NiagaraBool=false
     SpeedMin Value:NiagaraFloat=120
     SpeedMax Value:NiagaraFloat=180
     Speed Alpha:NiagaraFloat=$System.Speed_Alpha
    )
    Use Spawn Probability=false
   ) {enabled=false}
   &SpawnBurst_Instantaneous(Spawn Count=2, Spawn Group=4, Use Spawn Probability=false, UseLoopCountLimit=false) {module=SpawnBurst_Instantaneous004}
  ]
  particleSpawnScript [
   /SetVariables(Particles.RenderTag:NiagaraInt32=&SpawnGroup()) {isSetParameters=true}
   &VC_PaintedVariantsOverride(
    Black=<rgba 0.0191 0.0191 0.0191 3>
    BurntSienna=<rgba 0.1094 0.0232 0.0125 2>
    Cobalt=<rgba 0.0527 0.0622 0.5625 1>
    Crimson=<rgba 0.3385 0.0194 0.028 1>
    ForestGreen=<rgba 0 0.3073 0.0155 1>
    Grey=<rgba 0.0365 0.0365 0.0365 2>
    Lime=<rgba 0.1726 0.25 0.0104 1>
    Orange=<rgba 0.2552 0.089 0 1>
    Pink=<rgba 0.1771 0.0397 0.1271 1>
    Purple=<rgba 0.087 0.0482 0.388 1>
    R1 BurntSienna:LinearColor=<rgba 0.2188 0.0773 0.0627 1>
    R1 Cobalt:LinearColor=<rgba 0.0427 0.0521 0.5469 1>
    R1 Crimson:LinearColor=<rgba 0.4635 0.0266 0.0384 1>
    R1 DefaultColor=<rgba 1 1 0.2 1>
    R1 ForestGreen:LinearColor=<rgba 0.0942 0.3229 0.1057 1>
    R1 Lime:LinearColor=<rgba 0.244 0.2917 0.0395 1>
    R1 Orange:LinearColor=<rgba 0.5417 0.1889 0 1>
    R1 Purple:LinearColor=<rgba 0.103 0.0482 0.3372 1>
    R1 Saffron:LinearColor=<rgba 0.2865 0.2224 0.0119 1>
    R1 SkyBlue:LinearColor=<rgba 0 0.1975 0.2969 1>
    R1 TitaniumWhite:LinearColor=<rgba 0.2448 0.2448 0.2448 1>
    Saffron=<rgba 0.2188 0.1677 0 1>
    SkyBlue=<rgba 0 0.2079 0.3125 1>
    TitaniumWhite=<rgba 0.1302 0.1302 0.1302 1>
    Use R1 Painted Overrides=true
   )
   &InitializeParticle(
    Color=$Particles.Color
    Lifetime=$Emitter.CurrentLoopDuration
    Mesh Scale=&Multiply_VectorByFloat(Vector=<1 1 5>, Float=0.25)
    Mesh Scale Mode=::NewEnumerator4
    Sprite Rotation Angle=90
    Sprite Rotation Mode=::NewEnumerator0
    Sprite Size=&Multiply_Vector2D_ByFloat(Float=4, Vector2D=<20 6>)
    Sprite Size Mode=::NewEnumerator4
    UsePositionOffset=false
    Write Material Random=true
   )
   &PartitionParticles()
   &SkeletalMeshLocation(
    Bone / Socket Index=&Add_Integer(A=$Particles.PartitionParticles.Partition, B=4)
    Bone / Socket Sampling Mode=::NewEnumerator5
    Interpolate Mesh Sampling=false
    MaskSpawnGroup=false
    Mesh Sampling Type=::NewEnumerator6
    Sampled Position Offset=/VehicleCosmetics/FX/Niagara/Modules/MergeEmitters/VC_VectorBySpawnGroup(
     SpawnGroup or ExecIndex:NiagaraInt32=0
     CustomInt:NiagaraInt32=0
     MaximumSpawnGroups:NiagaraInt32=5
     Vector Group 0:Vector3f=<-50 0 -20>
     Vector Group 1:Vector3f=<-50 0 -20>
     Vector Group 2:Vector3f=<-50 0 -20>
     Vector Group 3:Vector3f=<-55 0 -20>
     Vector Group 4:Vector3f=<-80 0 -20>
     Vector Group 5:Vector3f=<0 0 0>
    )
    Skeletal Mesh=$System.SkeletalMesh
   )
   &AddVelocity(Velocity=&VC_Mask_Vectors(
     FrontEnd Value:Vector3f=<-1500 0 0>
     blsFrontEndPreview:NiagaraBool=false
     SpeedMin Value:Vector3f=<-800 0 0>
     SpeedMax Value:Vector3f=<-1 0 0>
     Speed Alpha:NiagaraFloat=$System.Speed_Alpha
    )) {enabled=false}
   &InitialMeshOrientation(Mesh Axis To Orient=<0 0 -1>, Mesh Orientation Mode=::NewEnumerator2, Orientation Vector=<0 0 0>, Use Rotation Vector=false)
   &SpriteFacingAndAlignment(Sprite Alignment=<1 0 0>, WriteAlignment=true, WriteFacing=false)
   /SetVariables(Particles.ColorTag1:LinearColor=<rgba 1 1 0.2 1>) {isSetParameters=true}
  ]
  particleUpdateScript [
   &FN_DisableTireFX(
    Mask Index Attribute:NiagaraInt32=$Particles.PartitionParticles.Partition
    Mask Index FL=2
    Mask Index FR=3
    Mask Index RL=0
    Mask Index RR=1
   )
   &ParticleState(Kill Particles When Lifetime Has Elapsed=false)
   &UpdateMeshOrientation(
    Rotation Rate=/VehicleCosmetics/FX/Niagara/Modules/MergeEmitters/VC_FloatBySpawnGroup(
     SpawnGroup or ExecIndex:NiagaraInt32=0
     CustomInt:NiagaraInt32=0
     MaximumSpawnGroups:NiagaraInt32=4
     Float SG 0:NiagaraFloat=0
     Float SG 1:NiagaraFloat=0
     Float SG 2:NiagaraFloat=0
     Float SG 3:NiagaraFloat=1
     Float SG 4:NiagaraFloat=0
     Float SG 5:NiagaraFloat=0
    )
    Rotation Vector=<0 0 0.4>
   )
   &SolveForcesAndVelocity(Clamp Velocity=false, Limit Acceleration=false)
   /SetVariables(
    Particles.Index1:Vector4f=&MakeVector4(X=1, Y=1, Z=0, W=0)
    Particles.MeshColor1:LinearColor=&VC_DynamicScaleColor(
     Color to Scale:LinearColor=&MakeCustomLinearColorFromBool(Bool=$User.bIsPainted, True Color=$Particles.Initial.Color, False Color=$Particles.Initial.ColorTag1)
     RGB Scale:Vector3f=&VectorFromFloat(Value=&MakeCustomFloatFromBool(Bool=$User.bIsPainted, True Float=5, False Float=1))
     Alpha Scale:NiagaraFloat=1
    )
    Particles.SpriteColor4:LinearColor=&VC_DynamicScaleColor(
     Color to Scale:LinearColor=$Particles.Initial.R1 Color
     RGB Scale:Vector3f=&VectorFromFloat(Value=&MakeCustomFloatFromBool(Bool=$User.bIsPainted, True Float=15, False Float=3))
     Alpha Scale:NiagaraFloat=1
    )
   ) {isSetParameters=true}
  ]
  rendererClasses [&NiagaraMeshRendererProperties, &NiagaraSpriteRendererProperties]
  renderers [
   {rendererClass=&NiagaraMeshRendererProperties, rendererIndex=0}
   {rendererClass=&NiagaraSpriteRendererProperties, rendererIndex=1}
  ]
 }
 NE_VC_DriftTrail_WheelWrap_Parent002 {
  emitterUpdateScript [
   &EmitterState(
    Inactive Response=::NewEnumerator1
    Life Cycle Mode=::NewEnumerator1
    Loop Behavior=::NewEnumerator1
    Loop Duration Mode=::NewEnumerator1
    UseLoopDelay=false
   )
   &SpawnBurst_Instantaneous(Spawn Count=2, Use Spawn Probability=false, UseLoopCountLimit=false)
  ]
  particleSpawnScript [
   /SetVariables(Particles.VisibilityTag=&SpawnGroup()) {enabled=false, isSetParameters=true}
   &VC_PaintedVariantsOverride(
    Black=<rgba 0.01 0.01 0.01 0.7913>
    BurntSienna=<rgba 0.3594 0.1093 0.0438 0.4522>
    Cobalt=<rgba 0 0.0065 0.349 1>
    Crimson=<rgba 0.2604 0 0.0071 1>
    ForestGreen=<rgba 0 0.1771 0.0089 0.7913>
    Grey=<rgba 0.2031 0.2031 0.2031 0.5652>
    Lime=<rgba 0.0937 0.1875 0 0.7304>
    Orange=<rgba 0.4375 0.1526 0 0.4522>
    Pink=<rgba 0.3021 0.0362 0.2054 0.7304>
    Purple=<rgba 0.1081 0.0886 0.3646 1>
    R1 Black:LinearColor=<rgba 0.1406 0.1406 0.1406 1>
    R1 BurntSienna:LinearColor=<rgba 0.7396 0.1892 0.131 0.5478>
    R1 Cobalt:LinearColor=<rgba 0.1667 0.1821 1 1>
    R1 Crimson:LinearColor=<rgba 0.6042 0.1227 0.1358 1>
    R1 ForestGreen:LinearColor=<rgba 0.1886 0.8229 0.2206 0.5304>
    R1 Lime:LinearColor=<rgba 0.4703 0.7604 0 0.5391>
    R1 Orange:LinearColor=<rgba 1 0.3302 0.125 0.5217>
    R1 Pink:LinearColor=<rgba 0.9896 0.3144 0.7441 0.5826>
    R1 Purple:LinearColor=<rgba 0.2027 0.1663 0.6795 1>
    R1 Saffron:LinearColor=<rgba 0.9323 0.7748 0.2574 0.5739>
    R1 SkyBlue:LinearColor=<rgba 0.1734 0.5293 0.7083 1>
    R1 TitaniumWhite:LinearColor=<rgba 0.901 0.901 0.901 0.6>
    R2 Black:LinearColor=<rgba 0.005 0.005 0.005 10>
    R2 BurntSienna:LinearColor=<rgba 0.352 0.0547 0.0218 1>
    R2 Purple:LinearColor=<rgba 0.2016 0.1651 0.6795 1>
    Saffron=<rgba 0.2292 0.1757 0 0.6957>
    SkyBlue=<rgba 0 0.2149 0.3229 0.6435>
    TitaniumWhite=<rgba 0.75 0.75 0.75 0.2609>
    Use R1 Painted Overrides=true
    Use R2 Painted Overrides=true
   )
   &InitializeParticle(
    Color=$Particles.Color
    Lifetime=$Emitter.CurrentLoopDuration
    Mesh Scale Mode=::NewEnumerator4
    Sprite Rotation Mode=::NewEnumerator2
    Sprite Size Mode=::NewEnumerator3
    UsePositionOffset=false
    Write Material Random=true
   )
   &PartitionParticles()
   &SkeletalMeshLocation(
    Bone / Socket Index=&Add_Integer(A=$Particles.PartitionParticles.Partition, B=4)
    Bone / Socket Sampling Mode=::NewEnumerator5
    Interpolate Mesh Sampling=false
    MaskSpawnGroup=false
    Mesh Sampling Type=::NewEnumerator6
    Skeletal Mesh=$System.SkeletalMesh
   )
   /SetVariables() {isSetParameters=true}
   &InitialMeshOrientation(Mesh Orientation Mode=::NewEnumerator7, Rotation=<0 0 0.5>, Use Rotation Vector=true)
  ]
  particleUpdateScript [
   &FN_DisableTireFX(
    Mask Index Attribute:NiagaraInt32=$Particles.PartitionParticles.Partition
    Mask Index FL=2
    Mask Index FR=3
    Mask Index RL=0
    Mask Index RR=1
   )
   &VC_WheelWrapScale(Wheel Index Attribute:NiagaraInt32=&Add_Integer(A=$Particles.PartitionParticles.Partition, B=2))
   &ParticleState(Kill Particles When Lifetime Has Elapsed=false)
   &UpdateMeshOrientation(Rotation Coordinate Space=::NewEnumerator2, Rotation Vector=<0 1 0>) {enabled=false}
   &SkeletalMeshLocation(
    Bone / Socket Index=&Add_Integer(A=$Particles.PartitionParticles.Partition, B=4)
    Bone / Socket Sampling Mode=::NewEnumerator5
    MaskSpawnGroup=false
    Mesh Sampling Type=::NewEnumerator6
    Orientation Sampling=::NewEnumerator1
    Skeletal Mesh=$System.SkeletalMesh
   ) {module=SkeletalMeshLocation001}
   &SolveForcesAndVelocity(Clamp Velocity=false, Limit Acceleration=false)
   /SetVariables(
    Particles.Index2:Vector4f=&MakeVector4(X=2, Y=1, Z=0, W=0)
    Particles.MeshColor0:LinearColor=&VC_DynamicScaleColor(
     Color to Scale:LinearColor=$Particles.Initial.Color
     RGB Scale:Vector3f=&MakeCustomVectorFromBool(
      Bool=$User.bIsPainted
      True Vector=&VectorFromFloat(Value=0.3)
      False Vector=&MakeVectorFromLinearColorRGB(Color=<rgba 0.0187 0.0079 0.125 1>)
     )
     Alpha Scale:NiagaraFloat=&MakeCustomFloatFromBool(Bool=$User.bIsPainted, True Float=1, False Float=0.5)
    )
    Particles.MeshColor2:LinearColor=&VC_DynamicScaleColor(
     Color to Scale:LinearColor=&MakeCustomLinearColorFromBool(Bool=$User.bIsPainted, True Color=$Particles.Initial.R2 Color, False Color=<rgba 1 1 0.2 1>)
     RGB Scale:Vector3f=&VectorFromFloat(Value=&MakeCustomFloatFromBool(Bool=$User.bIsPainted, True Float=80, False Float=30))
     Alpha Scale:NiagaraFloat=1
    )
    Particles.Meshcolor1:LinearColor=&VC_DynamicScaleColor(
     Color to Scale:LinearColor=&MakeCustomLinearColorFromBool(Bool=$User.bIsPainted, True Color=$Particles.Initial.R1 Color, False Color=<rgba 1 0.3 1 1>)
     RGB Scale:Vector3f=&VectorFromFloat(Value=&MakeCustomFloatFromBool(Bool=$User.bIsPainted, True Float=1, False Float=0.5))
     Alpha Scale:NiagaraFloat=&MakeCustomFloatFromBool(Bool=$User.bIsPainted, True Float=0.4, False Float=0.5)
    )
    Particles.index1:Vector4f=&MakeVector4(X=2, Y=10, Z=1, W=1)
   ) {isSetParameters=true}
   &DynamicMaterialParameters(
    Index 0 Param 1=2
    Index 0 Param 2=1
    Index 0 Param 3=1
    Index 0 Param 4=4
    Module.Index0Float0Write=true
    Module.Index0Float1Write=true
    Module.Index0Float2Write=true
    Module.Index0Float3Write=true
   )
  ]
  rendererClasses [&NiagaraMeshRendererProperties]
  renderers [
   {rendererClass=&NiagaraMeshRendererProperties, rendererIndex=0}
   {rendererClass=&NiagaraMeshRendererProperties, rendererIndex=1}
   {rendererClass=&NiagaraMeshRendererProperties, rendererIndex=2}
  ]
 }
}
summary {
 systemName=NS_DriftTrail_Magical_MP
 emitters [
  {
   emitterName=NE_VC_DriftTrail_Sprite_Parent
   rendererClasses [&NiagaraMeshRendererProperties, &NiagaraSpriteRendererProperties]
  }
  {emitterName=NE_VC_DriftTrail_Ribbon_Parent, rendererClasses [&NiagaraRibbonRendererProperties]}
  {emitterName=NE_VC_DriftTrail_WheelWrap_Parent002, rendererClasses [&NiagaraMeshRendererProperties]}
 ]
 userVariables [
  User.bIsFrontendPreview {type=NiagaraBool}
  User.bIsPainted {type=NiagaraBool}
  User.ColorVariantID {type=NiagaraInt32}
  User.offset {type=Vector3f, description="offset from wheels(set through cosmetic template)", defaultValue {struct=Vector3f, value=<0 0 0>}}
  User.PaintedColor {type=LinearColor, description="", defaultValue {struct=LinearColor, value=<rgba 1 1 1 1>}}
  User.PaintedScalar {type=NiagaraFloat, defaultValue=0, description=""}
  User.VehicleFlags {type=NiagaraInt32, defaultValue=806289408}
  User.Wheel_Radius_Array {
   type=NiagaraDataInterfaceArrayFloat
   defaultValue {
    struct=NiagaraExt_VariableValue_DataInterface
    value {
     dataInterface=&self:NiagaraDataInterfaceArrayFloat_1
     dataInterfaceClass=&NiagaraDataInterfaceArrayFloat
    }
   }
  }
  User.Wheel_Width_Array {
   type=NiagaraDataInterfaceArrayFloat
   defaultValue {
    struct=NiagaraExt_VariableValue_DataInterface
    value {
     dataInterface=&self:NiagaraDataInterfaceArrayFloat_0
     dataInterfaceClass=&NiagaraDataInterfaceArrayFloat
    }
   }
  }
  User.WheelSize {type=NiagaraFloat, defaultValue=1, description=""}
  User.WheelTouchingArray {
   type=NiagaraDataInterfaceArrayFloat
   defaultValue {
    struct=NiagaraExt_VariableValue_DataInterface
    value {
     dataInterface=&self:NiagaraDataInterfaceArrayFloat_2
     dataInterfaceClass=&NiagaraDataInterfaceArrayFloat
    }
   }
  }
 ]
}
